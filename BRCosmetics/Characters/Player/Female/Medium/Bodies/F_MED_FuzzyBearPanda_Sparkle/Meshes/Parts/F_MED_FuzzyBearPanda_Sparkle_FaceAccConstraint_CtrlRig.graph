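# Control Rig for the FuzzyBearPanda Sparkle female cosmetic that corrects hair master bones as the head moves.
# A spherical pose reader on head, relative to spine_05, outputs a weight from the head's orientation.
# That weight drives four chained ModifyTransforms that apply mirrored Z rotations to the left and right front hair and bangs bones.
# Three rotation constraints tying back hair to spine_05 are present but unconnected to execution.
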ControlRigBlueprint /BRCosmetics/Characters/Player/Female/Medium/Bodies/F_MED_FuzzyBearPanda_Sparkle/Meshes/Parts/F_MED_FuzzyBearPanda_Sparkle_FaceAccConstraint_CtrlRig
graphs [
 RigVMFunctionLibrary {links [], nodes []}
 RigVMModel {nodes [
   RigUnit_BeginExecution = &RigUnit_BeginExecution()
   n1 = &RigUnit_SphericalPoseReader(
    ActiveRegionScaleFactors=(PositiveWidth=0.25,NegativeWidth=0.25,PositiveHeight=0.25,NegativeHeight=0.25)
    ActiveRegionSize=0.1
    Debug=(bDrawDebug=false,DebugScale=25,DebugSegments=20,DebugThickness=0.25)
    DriverAxis=(X=0,Y=1,Z=0)
    DriverItem=(Type=Bone,Name="head")
    FalloffRegionScaleFactors=(PositiveWidth=0.75,NegativeWidth=0.5,PositiveHeight=4,NegativeHeight=4)
    FalloffSize=0.2
    OptionalParentItem=(Type=Bone,Name="spine_05")
    RotationOffset=(X=90,Y=0,Z=25)
    ExecutePin=@RigUnit_BeginExecution.ExecutePin
   )
   > &RigUnit_ModifyTransforms(
    ItemToModify=((Item=(Type=Bone,Name="dyn_hair_fr_l_MSTR"),Transform=(Rotation=(X=0,Y=0,Z=0,W=1),Translation=(X=0,Y=0,Z=0),Scale3D=(X=1,Y=1,Z=1))))
    ItemToModify.0.Transform.Rotation=&RigVMFunction_MathQuaternionFromEuler(Euler=(X=0,Y=0,Z=-15)).Result
    Weight=@n1.OutputParam
   )
   > &RigUnit_ModifyTransforms(
    ItemToModify=((Item=(Type=Bone,Name="dyn_hair_fr_r_MSTR"),Transform=(Rotation=(X=0,Y=0,Z=0,W=1),Translation=(X=0,Y=0,Z=0),Scale3D=(X=1,Y=1,Z=1))))
    ItemToModify.0.Transform.Rotation=&RigVMFunction_MathQuaternionFromEuler(Euler=(X=0,Y=0,Z=15)).Result
    Weight=@n1.OutputParam
   )
   > &RigUnit_ModifyTransforms(
    ItemToModify=((Item=(Type=Bone,Name="dyn_hairbangs_fr_l_MSTR"),Transform=(Rotation=(X=0,Y=0,Z=0,W=1),Translation=(X=0,Y=0,Z=0),Scale3D=(X=1,Y=1,Z=1))))
    ItemToModify.0.Transform.Rotation=&RigVMFunction_MathQuaternionFromEuler(Euler=(X=0,Y=0,Z=-15)).Result
    Weight=@n1.OutputParam
   )
   > &RigUnit_ModifyTransforms(
    ItemToModify=((Item=(Type=Bone,Name="dyn_hairbangs_fr_r_MSTR"),Transform=(Rotation=(X=0,Y=0,Z=0,W=1),Translation=(X=0,Y=0,Z=0),Scale3D=(X=1,Y=1,Z=1))))
    ItemToModify.0.Transform.Rotation=&RigVMFunction_MathQuaternionFromEuler(Euler=(X=0,Y=0,Z=15)).Result
    Weight=@n1.OutputParam
   )
   &RigUnit_RotationConstraintLocalSpaceOffset(
    AdvancedSettings=(RotationOrderForFilter=XZY)
    Child=(Type=Bone,Name="dyn_hair_bk_l_MSTR")
    ExecutePin=()
    Filter=(bX=True,bY=True,bZ=True)
    Parents=((Item=(Type=Bone,Name="spine_05"),Weight=1))
    Weight=1
   )
   &RigUnit_RotationConstraintLocalSpaceOffset(
    AdvancedSettings=(RotationOrderForFilter=XZY)
    Child=(Type=Bone,Name="dyn_hair_bk_r_MSTR")
    ExecutePin=()
    Filter=(bX=True,bY=True,bZ=True)
    Parents=((Item=(Type=Bone,Name="spine_05"),Weight=1))
    Weight=1
   )
   &RigUnit_RotationConstraintLocalSpaceOffset(
    AdvancedSettings=(RotationOrderForFilter=XZY)
    Child=(Type=Bone,Name="dyn_hair_bk_MSTR")
    ExecutePin=()
    Filter=(bX=True,bY=True,bZ=True)
    Parents=((Item=(Type=Bone,Name="spine_05"),Weight=1))
    Weight=1
   )
  ]}
]
hierarchy [
 root {loc=<0 0 0>, rot=<0 0 0 1>}
 pelvis {parent=root, loc=<0 -0.5301 98.0571>, rot=<0.0127 0.707 -0.0127 -0.707>}
 spine_01 {parent=pelvis, loc=<6.1208 -0.9328 0>, rot=<0 0 0.034 -0.9994>}
 spine_02 {parent=spine_01, loc=<6.5996 -0.2095 0>, rot=<0 0 0.0238 -0.9997>}
 spine_03 {parent=spine_02, loc=<6.4914 -0.1237 0>, rot=<0 0 0.0149 -0.9999>}
 spine_04 {parent=spine_03, loc=<6.9598 0.0794 0>, rot=<0 0 -0.0132 -0.9999>}
 spine_05 {parent=spine_04, loc=<6.9987 0 0>, rot=<0 0 -0.0015 -1>}
 attach_cape {parent=spine_05, loc=<9.3216 -8.4501 0>, rot=<0.0616 -0.7044 -0.0616 -0.7044>}
 neck_01 {parent=spine_05, loc=<11.3138 0.1529 0>, rot=<0 0 -0.2003 -0.9797>}
 neck_02 {parent=neck_01, loc=<4.5627 0 0>, rot=<0 0 0.0504 -0.9987>}
 head {parent=neck_02, loc=<5.3914 0 0>, rot=<0 0 0.115 -0.9934>}
 dyn_hair_bk_l_MSTR {parent=head, loc=<9.3157 -3.736 -6.883>, rot=<0 -0.7071 0.7071 0>}
 dyn_hair_bk_01_l {parent=dyn_hair_bk_l_MSTR, loc=<0 0 0>, rot=<0 0.0358 0 -0.9994>}
 dyn_hair_bk_02_l {parent=dyn_hair_bk_01_l, loc=<8.8667 0 0>, rot=<0 0.0681 0 -0.9977>}
 dyn_hair_bk_03_l {parent=dyn_hair_bk_02_l, loc=<8.8667 0 0>, rot=<0 0.2025 0 -0.9793>}
 dyn_hair_bk_04_l {parent=dyn_hair_bk_03_l, loc=<8.8667 0 0>, rot=<0 -0.1322 0 -0.9912>}
 dyn_hair_bk_05_l {parent=dyn_hair_bk_04_l, loc=<8.8667 0 0>, rot=<0 -0.1534 0 -0.9882>}
 dyn_hair_bk_06_l {parent=dyn_hair_bk_05_l, loc=<8.8667 0 0>, rot=<0 0 0 1>}
 dyn_hair_bk_MSTR {parent=head, loc=<8.5295 -9.1964 0>, rot=<0 -0.7071 0.7071 0>}
 dyn_hair_bk_01 {parent=dyn_hair_bk_MSTR, loc=<0 0 0>, rot=<0 0.1043 0 -0.9945>}
 dyn_hair_bk_02 {parent=dyn_hair_bk_01, loc=<9.0292 0 0>, rot=<0 0.0631 0 -0.998>}
 dyn_hair_bk_03 {parent=dyn_hair_bk_02, loc=<9.0292 0 0>, rot=<0 -0.0099 0 -1>}
 dyn_hair_bk_04 {parent=dyn_hair_bk_03, loc=<9.0292 0 0>, rot=<0 -0.0837 0 -0.9965>}
 dyn_hair_bk_05 {parent=dyn_hair_bk_04, loc=<9.0292 0 0>, rot=<0 -0.1329 0 -0.9911>}
 dyn_hair_bk_06 {parent=dyn_hair_bk_05, loc=<9.0292 0 0>, rot=<0 -0.0055 0 -1>}
 dyn_hair_r_MSTR {parent=head, loc=<8.6113 -1.0761 6.2339>, rot=<0.0611 -0.0752 0.9911 -0.0917>}
 dyn_hair_01_r {parent=dyn_hair_r_MSTR, loc=<0 0 0>, rot=<0 0 0 1>}
 dyn_hair_02_r {parent=dyn_hair_01_r, loc=<8.2913 -0.1423 0>, rot=<0 0 0 1>}
 dyn_hair_03_r {parent=dyn_hair_02_r, loc=<7.6813 0 0.0001>, rot=<0 0 0 1>}
 dyn_hair_bk_r_MSTR {parent=head, loc=<9.316 -3.736 6.883>, rot=<0 -0.7071 0.7071 0>}
 dyn_hair_bk_01_r {parent=dyn_hair_bk_r_MSTR, loc=<0 0 0>, rot=<0 0.0358 0 -0.9994>}
 dyn_hair_bk_02_r {parent=dyn_hair_bk_01_r, loc=<8.8668 0 0>, rot=<0 0.0681 0 -0.9977>}
 dyn_hair_bk_03_r {parent=dyn_hair_bk_02_r, loc=<8.8672 0 0>, rot=<0 0.2026 0 -0.9793>}
 dyn_hair_bk_04_r {parent=dyn_hair_bk_03_r, loc=<8.8661 0 0>, rot=<0 -0.1322 0 -0.9912>}
 dyn_hair_bk_05_r {parent=dyn_hair_bk_04_r, loc=<8.8668 0 0>, rot=<0 -0.1534 0 -0.9882>}
 dyn_hair_bk_06_r {parent=dyn_hair_bk_05_r, loc=<8.8673 0 0>, rot=<0 0 0 -1>}
 dyn_hair_fr_r_MSTR {parent=head, loc=<7.5735 3.9797 5.9941>, rot=<0.0608 -0.0014 0.9976 -0.0317>}
 dyn_hair_fr_01_r {parent=dyn_hair_fr_r_MSTR, loc=<0 0 0>, rot=<0 0 0 1>}
 dyn_hair_fr_02_r {parent=dyn_hair_fr_01_r, loc=<8.2913 -0.1423 0>, rot=<-0.0004 -0.0382 0.0096 -0.9992>}
 dyn_hair_fr_03_r {parent=dyn_hair_fr_02_r, loc=<7.6813 0 0.0001>, rot=<-0.0046 -0.0334 0.1366 -0.9901>}
 dyn_hair_fr_04_r {parent=dyn_hair_fr_03_r, loc=<6.2351 -0.0001 0>, rot=<0 0 0 -1>}
 dyn_hair_fr_l_MSTR {parent=head, loc=<7.5731 3.9797 -5.9941>, rot=<0.0317 -0.9976 0.0014 -0.0608>}
 dyn_hair_fr_01_l {parent=dyn_hair_fr_l_MSTR, loc=<0 0 0>, rot=<0 0 0 1>}
 dyn_hair_fr_02_l {parent=dyn_hair_fr_01_l, loc=<8.291 0.1423 0>, rot=<0.0004 -0.0382 -0.0096 -0.9992>}
 dyn_hair_fr_03_l {parent=dyn_hair_fr_02_l, loc=<7.681 0 -0.0001>, rot=<0.0046 -0.0334 -0.1366 -0.9901>}
 dyn_hair_fr_04_l {parent=dyn_hair_fr_03_l, loc=<6.235 0 0>, rot=<0 0 0 -1>}
 dyn_hair_l_MSTR {parent=head, loc=<8.6117 -1.0761 -6.2339>, rot=<0.0917 -0.9911 0.0752 -0.0611>}
 dyn_hair_01_l {parent=dyn_hair_l_MSTR, loc=<0 0 0>, rot=<0 0 0 1>}
 dyn_hair_02_l {parent=dyn_hair_01_l, loc=<8.292 0.1424 0.0001>, rot=<0 0 0 1>}
 dyn_hair_03_l {parent=dyn_hair_02_l, loc=<7.681 0 0>, rot=<0 0 0 1>}
 dyn_hairbangs_fr_r_MSTR {parent=head, loc=<11.012 7.594 2.772>, rot=<0.2002 0.003 0.9792 -0.0316>}
 dyn_hairbangs_fr_01_r {parent=dyn_hairbangs_fr_r_MSTR, loc=<0 0 0>, rot=<0 0 0 1>}
 dyn_hairbangs_fr_02_r {parent=dyn_hairbangs_fr_01_r, loc=<8.2913 -0.1423 0>, rot=<0.0043 -0.1819 0.042 -0.9824>}
 dyn_hairbangs_fr_03_r {parent=dyn_hairbangs_fr_02_r, loc=<7.6813 0 0.0001>, rot=<-0.033 -0.2394 0.1326 -0.9613>}
 dyn_hairbangs_fr_04_r {parent=dyn_hairbangs_fr_03_r, loc=<6.2351 -0.0001 0>, rot=<0 0 0 -1>}
 dyn_hairbangs_fr_l_MSTR {parent=head, loc=<11.012 7.594 -2.7717>, rot=<0.0316 -0.9792 -0.003 -0.2002>}
 dyn_hairbangs_fr_01_l {parent=dyn_hairbangs_fr_l_MSTR, loc=<0 0 0>, rot=<0 0 0 1>}
 dyn_hairbangs_fr_02_l {parent=dyn_hairbangs_fr_01_l, loc=<8.291 0.1423 0>, rot=<-0.0043 -0.1819 -0.042 -0.9824>}
 dyn_hairbangs_fr_03_l {parent=dyn_hairbangs_fr_02_l, loc=<7.681 0 -0.0001>, rot=<0.033 -0.2394 -0.1326 -0.9613>}
 dyn_hairbangs_fr_04_l {parent=dyn_hairbangs_fr_03_l, loc=<6.235 0 0>, rot=<0 0 0 -1>}
 attach_backpack {parent=spine_05, loc=<0 -8.35 0>, rot=<-0.0093 -0.707 0.0093 -0.707>}
 clavicle_l {parent=spine_05, loc=<5.4831 -0.1053 -3.2946>, rot=<0.0465 -0.7289 -0.009 -0.6829>}
 upperarm_l {parent=clavicle_l, loc=<9.8658 -1.2136 -0.4105>, rot=<0.0164 -0.3245 -0.0269 -0.9454>}
 upperarm_twist_01_l {parent=upperarm_l, loc=<7.2135 -0.1134 0>, rot=<0 0 0 1>}
 clavicle_r {parent=spine_05, loc=<5.4831 -0.1053 3.2946>, rot=<0.7289 0.0465 -0.6829 0.009>}
 upperarm_r {parent=clavicle_r, loc=<-9.8658 1.2136 0.4105>, rot=<0.0164 -0.3245 -0.0269 -0.9454>}
 upperarm_twist_01_r {parent=upperarm_r, loc=<-7.2135 0.1134 0>, rot=<0 0 0 -1>}
]
