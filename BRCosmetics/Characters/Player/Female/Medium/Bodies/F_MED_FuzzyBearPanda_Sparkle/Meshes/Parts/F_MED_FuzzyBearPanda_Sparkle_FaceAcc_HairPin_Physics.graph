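# Physics asset for the FuzzyBearPanda Sparkle female cosmetic's hairpin face accessory part, covering dynamic hair bones and the upper body.
# Three six-segment back-hair chains hang from small MSTR anchor bodies on the head, while front strands and bangs attach directly to the head.
# Hair joints use swing limits with very tight twist.
# Capsules on pelvis, spine, neck, head, clavicles and upper arms form a body proxy linked by free-angular constraints.
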PhysicsAsset /BRCosmetics/Characters/Player/Female/Medium/Bodies/F_MED_FuzzyBearPanda_Sparkle/Meshes/Parts/F_MED_FuzzyBearPanda_Sparkle_FaceAcc_HairPin_Physics
bodies {
 clavicle_l [{shapeType=Capsule, length=11, radius=5.25, shapeName=clavicle_l_capsule, center=<3.5 -1 0>, rotation=<rot 80 -0.0003 -0.0003>}]
 clavicle_r [{shapeType=Capsule, length=11, radius=5.25, shapeName=clavicle_r_capsule, center=<-3.5 1 0>, rotation=<rot 80 -0.0003 -0.0004>}]
 dyn_hair_bk_01 [{shapeType=Capsule, length=5, radius=3, shapeName=dyn_hair_bk_01_capsule, center=<5 0 0>, rotation=<rot 90 0 0>}]
 dyn_hair_bk_01_l [{shapeType=Capsule, length=5, radius=3, shapeName=dyn_hair_bk_01_l_capsule, center=<5 0 0>, rotation=<rot 90 0 0>}]
 dyn_hair_bk_01_r [{shapeType=Capsule, length=5, radius=3, shapeName=dyn_hair_bk_01_r_capsule, center=<5 0 0>, rotation=<rot 90 0 0>}]
 dyn_hair_bk_02 [{shapeType=Capsule, length=5, radius=4.5, shapeName=dyn_hair_bk_02_capsule, center=<5 0 1>, rotation=<rot 90 0 0>}]
 dyn_hair_bk_02_l [{shapeType=Capsule, length=5, radius=4.5, shapeName=dyn_hair_bk_02_l_capsule, center=<5 -1 1>, rotation=<rot 80.0001 180 170>}]
 dyn_hair_bk_02_r [{shapeType=Capsule, length=5, radius=4.5, shapeName=dyn_hair_bk_02_r_capsule, center=<5 1 1>, rotation=<rot 80.0001 -180 -170>}]
 dyn_hair_bk_03 [{shapeType=Capsule, length=9, radius=4, shapeName=dyn_hair_bk_03_capsule, center=<5 0 -0.5>, rotation=<rot 0 0 90>}]
 dyn_hair_bk_03_l [{shapeType=Capsule, length=9, radius=4, shapeName=dyn_hair_bk_03_l_capsule, center=<5 -2 3>, rotation=<rot -0.0014 0.0001 120>}]
 dyn_hair_bk_03_r [{shapeType=Capsule, length=9, radius=4, shapeName=dyn_hair_bk_03_r_capsule, center=<5 2 3>, rotation=<rot -0.0006 -0.0001 50>}]
 dyn_hair_bk_04 [{shapeType=Capsule, length=11, radius=3, shapeName=dyn_hair_bk_04_capsule, center=<5 0 -0.5>, rotation=<rot 0 0 90>}]
 dyn_hair_bk_04_l [{shapeType=Capsule, length=11, radius=3, shapeName=dyn_hair_bk_04_l_capsule, center=<5 -2 2.5>, rotation=<rot -0.0003 0 110>}]
 dyn_hair_bk_04_r [{shapeType=Capsule, length=11, radius=3, shapeName=dyn_hair_bk_04_r_capsule, center=<5 2 2.5>, rotation=<rot -0.0003 0 60>}]
 dyn_hair_bk_05 [{shapeType=Capsule, length=13, radius=2, shapeName=dyn_hair_bk_05_capsule, center=<5 0 -0.5>, rotation=<rot 0 0 90>}]
 dyn_hair_bk_05_l [{shapeType=Capsule, length=13, radius=2, shapeName=dyn_hair_bk_05_l_capsule, center=<5 -2 1.5>, rotation=<rot -0.0001 0 110>}]
 dyn_hair_bk_05_r [{shapeType=Capsule, length=13, radius=2, shapeName=dyn_hair_bk_05_r_capsule, center=<5 2 1.5>, rotation=<rot -0.0004 0 60>}]
 dyn_hair_bk_06 [{shapeType=Capsule, length=12, radius=2.5, shapeName=dyn_hair_bk_06_capsule, center=<4 0 -2>, rotation=<rot 0 0 90>}]
 dyn_hair_bk_06_l [{shapeType=Capsule, length=12, radius=2.5, shapeName=dyn_hair_bk_06_l_capsule, center=<4 -2 -0.5>, rotation=<rot -6.7181 18.8817 108.8816>}]
 dyn_hair_bk_06_r [{shapeType=Capsule, length=12, radius=2.5, shapeName=dyn_hair_bk_06_r_capsule, center=<4 2 -0.5>, rotation=<rot -6.7181 -18.8817 71.1184>}]
 dyn_hair_bk_MSTR [{shapeType=Capsule, length=0.505, radius=0.505, shapeName=dyn_hair_bk_MSTR_capsule, rotation=<rot 0 -90 0>}]
 dyn_hair_bk_l_MSTR [{shapeType=Capsule, length=0.505, radius=0.505, shapeName=dyn_hair_bk_l_MSTR_capsule, rotation=<rot 0 -90 0>}]
 dyn_hair_bk_r_MSTR [{shapeType=Capsule, length=0.505, radius=0.505, shapeName=dyn_hair_bk_r_MSTR_capsule, rotation=<rot 0 -90 0>}]
 dyn_hair_fr_02_l [{shapeType=Capsule, length=5, radius=1, shapeName=dyn_hair_fr_02_l_capsule, center=<4 -0.5 0>, rotation=<rot 0 0.0003 120>}]
 dyn_hair_fr_02_r [{shapeType=Capsule, length=4, radius=1.75, shapeName=dyn_hair_fr_02_r_capsule, center=<2.2188 0.55 0>, rotation=<rot 0 -0.0003 80>}]
 dyn_hair_fr_03_l [{shapeType=Capsule, length=5, radius=2.5, shapeName=dyn_hair_fr_03_l_capsule, center=<3.15 0 0.55>, rotation=<rot 90 178.2101 168.2099>}]
 dyn_hair_fr_03_r [{shapeType=Capsule, length=4, radius=1.75, shapeName=dyn_hair_fr_03_r_capsule, center=<2 0 0>, rotation=<rot 90 0 0>}]
 dyn_hair_fr_04_l [{shapeType=Capsule, length=5, radius=1.5, shapeName=dyn_hair_fr_04_l_capsule, center=<4.5 1.6 1.2>, rotation=<rot 70 -169.9979 159.9999>}]
 dyn_hair_fr_04_r [{shapeType=Capsule, length=12, radius=2.5, shapeName=dyn_hair_fr_04_r_capsule, center=<6 -2.13 2.98>, rotation=<rot 70 169.9996 -170>}]
 dyn_hairbangs_fr_02_l [{shapeType=Capsule, length=5, radius=1.3, shapeName=dyn_hairbangs_fr_02_l_capsule, center=<4.5 0 0>, rotation=<rot -90 0 0>}]
 dyn_hairbangs_fr_02_r [{shapeType=Capsule, length=5, radius=1.3, shapeName=dyn_hairbangs_fr_02_r_capsule, center=<4.5 0 0>, rotation=<rot -90 0 0>}]
 dyn_hairbangs_fr_03_l [{shapeType=Capsule, length=4, radius=1.5, shapeName=dyn_hairbangs_fr_03_l_capsule, center=<4.5 0 0>, rotation=<rot -90 0 0>}]
 dyn_hairbangs_fr_03_r [{shapeType=Capsule, length=4, radius=1.5, shapeName=dyn_hairbangs_fr_03_r_capsule, center=<4.5 0 0>, rotation=<rot -90 0 0>}]
 head [
  {shapeType=Capsule, length=6.3, radius=5.5, shapeName=head_capsule, center=<3.9207 1.2637 0>, rotation=<rot 0 0 90>}
  {shapeType=Capsule, length=2, radius=5.8, shapeName=head_capsule, center=<-0.0503 3.6554 0>, rotation=<rot 90 0 0>}
 ]
 neck_01 [{shapeType=Capsule, length=2, radius=4.5, shapeName=neck_01_capsule, center=<3.6683 -0.2876 0>, rotation=<rot 90 0 0>}]
 pelvis [{shapeType=Capsule, length=3, radius=13, shapeName=pelvis_capsule, center=<2.3528 0 0>}]
 spine_03 [{shapeType=Capsule, length=10, radius=12, shapeName=spine_04_capsule, center=<0.229 3.0625 0>, rotation=<rot 90 0 10.0001>}]
 spine_05 [{shapeType=Capsule, length=5, radius=10.2, shapeName=spine_05_capsule, center=<-1.2254 -0.2839 0>}]
 upperarm_l [{shapeType=Capsule, length=20, radius=3.5, shapeName=upperarm_l_capsule, center=<12 -0.5 0>, rotation=<rot 90 0 0>}]
 upperarm_r [{shapeType=Capsule, length=20, radius=3.5, shapeName=upperarm_r_capsule, center=<-12 0.5 0>, rotation=<rot 90 0 0>}]
}
constraints [
 {bone1Name=dyn_hair_bk_02, bone2Name=dyn_hair_bk_01, swing1LimitDegrees=23, swing2LimitDegrees=15, twistLimitDegrees=2}
 {bone1Name=dyn_hair_bk_03, bone2Name=dyn_hair_bk_02, swing1LimitDegrees=23, swing2LimitDegrees=35, twistLimitDegrees=2}
 {bone1Name=dyn_hair_bk_04, bone2Name=dyn_hair_bk_03, swing1LimitDegrees=35, swing2LimitDegrees=35, twistLimitDegrees=2}
 {bone1Name=dyn_hair_bk_05, bone2Name=dyn_hair_bk_04, swing1LimitDegrees=35, swing2LimitDegrees=35, twistLimitDegrees=2}
 {bone1Name=dyn_hair_bk_06, bone2Name=dyn_hair_bk_05, swing1LimitDegrees=35, swing2LimitDegrees=35, twistLimitDegrees=2}
 {bone1Name=dyn_hair_bk_02_l, bone2Name=dyn_hair_bk_01_l, swing1LimitDegrees=23, swing2LimitDegrees=15, twistLimitDegrees=2}
 {bone1Name=dyn_hair_bk_03_l, bone2Name=dyn_hair_bk_02_l, swing1LimitDegrees=23, swing2LimitDegrees=35, twistLimitDegrees=2}
 {bone1Name=dyn_hair_bk_04_l, bone2Name=dyn_hair_bk_03_l, swing1LimitDegrees=35, swing2LimitDegrees=35, twistLimitDegrees=2}
 {bone1Name=dyn_hair_bk_05_l, bone2Name=dyn_hair_bk_04_l, swing1LimitDegrees=35, swing2LimitDegrees=35, twistLimitDegrees=2}
 {bone1Name=dyn_hair_bk_06_l, bone2Name=dyn_hair_bk_05_l, swing1LimitDegrees=35, swing2LimitDegrees=35, twistLimitDegrees=2}
 {bone1Name=dyn_hair_bk_02_r, bone2Name=dyn_hair_bk_01_r, swing1LimitDegrees=23, swing2LimitDegrees=15, twistLimitDegrees=2}
 {bone1Name=dyn_hair_bk_03_r, bone2Name=dyn_hair_bk_02_r, swing1LimitDegrees=23, swing2LimitDegrees=35, twistLimitDegrees=2}
 {bone1Name=dyn_hair_bk_04_r, bone2Name=dyn_hair_bk_03_r, swing1LimitDegrees=35, swing2LimitDegrees=35, twistLimitDegrees=2}
 {bone1Name=dyn_hair_bk_05_r, bone2Name=dyn_hair_bk_04_r, swing1LimitDegrees=35, swing2LimitDegrees=35, twistLimitDegrees=2}
 {bone1Name=dyn_hair_bk_06_r, bone2Name=dyn_hair_bk_05_r, swing1LimitDegrees=35, swing2LimitDegrees=35, twistLimitDegrees=2}
 {bone1Name=dyn_hair_fr_02_l, bone2Name=head, swing1LimitDegrees=35, swing2LimitDegrees=35, twistLimitDegrees=2}
 {bone1Name=dyn_hair_fr_03_l, bone2Name=dyn_hair_fr_02_l, swing1LimitDegrees=35, swing2LimitDegrees=35, twistLimitDegrees=2}
 {bone1Name=dyn_hair_fr_04_l, bone2Name=dyn_hair_fr_03_l, swing1LimitDegrees=35, swing2LimitDegrees=35, twistLimitDegrees=2}
 {bone1Name=dyn_hair_fr_02_r, bone2Name=head, swing1LimitDegrees=25, swing2LimitDegrees=25, twistLimitDegrees=2}
 {bone1Name=dyn_hair_fr_03_r, bone2Name=dyn_hair_fr_02_r, swing1LimitDegrees=25, swing2LimitDegrees=25, twistLimitDegrees=2}
 {bone1Name=dyn_hair_fr_04_r, bone2Name=dyn_hair_fr_03_r, swing1LimitDegrees=25, swing2LimitDegrees=25, twistLimitDegrees=2}
 {bone1Name=dyn_hairbangs_fr_02_l, bone2Name=head, swing1LimitDegrees=35, swing2LimitDegrees=35, twistLimitDegrees=2}
 {bone1Name=dyn_hairbangs_fr_03_l, bone2Name=dyn_hairbangs_fr_02_l, swing1LimitDegrees=35, swing2LimitDegrees=35, twistLimitDegrees=2}
 {bone1Name=dyn_hairbangs_fr_02_r, bone2Name=head, swing1LimitDegrees=35, swing2LimitDegrees=35, twistLimitDegrees=2}
 {bone1Name=dyn_hairbangs_fr_03_r, bone2Name=dyn_hairbangs_fr_02_r, swing1LimitDegrees=35, swing2LimitDegrees=35, twistLimitDegrees=2}
 {
  bone1Name=head
  bone2Name=neck_01
  swing1LimitDegrees=45
  swing1Motion=Free
  swing2LimitDegrees=45
  swing2Motion=Free
  twistLimitDegrees=45
  twistMotion=Free
 }
 {
  bone1Name=neck_01
  bone2Name=spine_05
  swing1LimitDegrees=45
  swing1Motion=Free
  swing2LimitDegrees=45
  swing2Motion=Free
  twistLimitDegrees=45
  twistMotion=Free
 }
 {
  bone1Name=clavicle_l
  bone2Name=spine_05
  swing1LimitDegrees=45
  swing1Motion=Free
  swing2LimitDegrees=45
  swing2Motion=Free
  twistLimitDegrees=45
  twistMotion=Free
 }
 {
  bone1Name=clavicle_r
  bone2Name=spine_05
  swing1LimitDegrees=45
  swing1Motion=Free
  swing2LimitDegrees=45
  swing2Motion=Free
  twistLimitDegrees=45
  twistMotion=Free
 }
 {
  bone1Name=spine_05
  bone2Name=spine_03
  swing1LimitDegrees=45
  swing1Motion=Free
  swing2LimitDegrees=45
  swing2Motion=Free
  twistLimitDegrees=45
  twistMotion=Free
 }
 {
  bone1Name=upperarm_r
  bone2Name=clavicle_r
  swing1LimitDegrees=45
  swing1Motion=Free
  swing2LimitDegrees=45
  swing2Motion=Free
  twistLimitDegrees=45
  twistMotion=Free
 }
 {
  bone1Name=upperarm_l
  bone2Name=clavicle_l
  swing1LimitDegrees=45
  swing1Motion=Free
  swing2LimitDegrees=45
  swing2Motion=Free
  twistLimitDegrees=45
  twistMotion=Free
 }
 {
  bone1Name=spine_03
  bone2Name=pelvis
  swing1LimitDegrees=45
  swing1Motion=Free
  swing2LimitDegrees=45
  swing2Motion=Free
  twistLimitDegrees=45
  twistMotion=Free
 }
 {
  bone1Name=dyn_hair_bk_MSTR
  bone2Name=head
  swing1LimitDegrees=45
  swing1Motion=Free
  swing2LimitDegrees=45
  swing2Motion=Free
  twistLimitDegrees=45
  twistMotion=Free
 }
 {
  bone1Name=dyn_hair_bk_l_MSTR
  bone2Name=head
  swing1LimitDegrees=45
  swing1Motion=Free
  swing2LimitDegrees=45
  swing2Motion=Free
  twistLimitDegrees=45
  twistMotion=Free
 }
 {
  bone1Name=dyn_hair_bk_r_MSTR
  bone2Name=head
  swing1LimitDegrees=45
  swing1Motion=Free
  swing2LimitDegrees=45
  swing2Motion=Free
  twistLimitDegrees=45
  twistMotion=Free
 }
 {bone1Name=dyn_hair_bk_01_r, bone2Name=dyn_hair_bk_r_MSTR, swing1LimitDegrees=2, swing2LimitDegrees=15, twistLimitDegrees=2}
 {bone1Name=dyn_hair_bk_01_l, bone2Name=dyn_hair_bk_l_MSTR, swing1LimitDegrees=2, swing2LimitDegrees=15, twistLimitDegrees=2}
 {bone1Name=dyn_hair_bk_01, bone2Name=dyn_hair_bk_MSTR, swing1LimitDegrees=2, swing2LimitDegrees=15, twistLimitDegrees=2}
]
