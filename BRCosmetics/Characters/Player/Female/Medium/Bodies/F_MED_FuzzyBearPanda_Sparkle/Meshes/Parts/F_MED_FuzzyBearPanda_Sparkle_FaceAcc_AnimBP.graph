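# Post-process style AnimBP for the FuzzyBearPanda Sparkle face accessory, animating back hair over copied body poses and Control Rigs.
# With backbling, it simulates the HairPin physics asset in head space and constrains back-hair bones to neck and spine with fixed offsets.
# Otherwise, it runs spine-space RigidBody physics, adds melee, skydive and airborne bone offsets, then a pistol post-sim rig.
# The EventGraph derives alphas from movement, weapon and backbling state.
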AnimBlueprint /BRCosmetics/Characters/Player/Female/Medium/Bodies/F_MED_FuzzyBearPanda_Sparkle/Meshes/Parts/F_MED_FuzzyBearPanda_Sparkle_FaceAcc_AnimBP
target_skeleton=/Game/Characters/Player/Female/Medium/Bodies/F_MED_FuzzyBearPanda_Sparkle/Meshes/Parts/F_MED_FuzzyBearPanda_Sparkle_FaceAcc_Skeleton
graphs [
 AnimGraph {nodes [
   n1 = Control Rig(Source=Layered blend per bone(BasePose=Copy Pose From Mesh() {bCopyCurves=true, bUseAttachedParent=true}, BlendPoses_0=Copy Pose From Mesh()) {layerSetup [{branchFilters [{boneName=upperarm_l}, {boneName=upperarm_r}]}]}) {
    alphaBoolBlend {blendOption=Linear}
    controlRigAssetReference {blueprintRigClass=&/F_MED_FuzzyBearPanda_Sparkle_FaceAcc_CtrlRig.F_MED_FuzzyBearPanda_Sparkle_FaceAcc_CtrlRig_C}
   }
   n2 = Control Rig(Source=@n1, Alpha=Get Sky Dive Control Rig Alpha() -> (Sky Dive Control Rig Alpha:double)) {
    alphaBoolBlend {blendOption=Linear}
    controlRigAssetReference {blueprintRigClass=&/F_MED_FuzzyBearPanda_Sparkle_FaceAccConstraint_CtrlRig.F_MED_FuzzyBearPanda_Sparkle_FaceAccConstraint_CtrlRig_C}
   }
   n3 = Control Rig(Source=@n2, Alpha=Get Is Pistol Targeting Alpha() -> (Is Pistol Targeting Alpha:float)) {
    alphaBoolBlend {blendOption=Linear}
    controlRigAssetReference {blueprintRigClass=&/F_MED_FuzzyBearPanda_Sparkle_FaceAcc_Pistol_CtrlRig.F_MED_FuzzyBearPanda_Sparkle_FaceAcc_Pistol_CtrlRig_C}
   }
   n4 = RigidBody!(
    ComponentPose=Local To Component(LocalPose=@n3)
    Alpha=Get Backbling Physics Alpha() -> (Backbling Physics Alpha:float)
    OverrideWorldGravity:Vector=Get GravityOverride()
    ExternalForce
   ) -> (Pose) {
    bClampLinearTranslationLimitToRefPose=true
    bOverrideWorldGravity=true
    overridePhysicsAsset=&/F_MED_FuzzyBearPanda_Sparkle_FaceAcc_HairPin_Physics
    simulationSpace=BaseBoneSpace
    alphaBoolBlend {blendOption=Linear}
    alphaScaleBiasClamp {bInterpResult=true}
    baseBoneRef {boneName=head}
    componentAppliedLinearAccClamp=<3000 3000 3000>
    componentLinearAccScale=<0.4 0.3 0.3>
    componentLinearVelScale=<0.4 0.3 0.3>
    simSpaceSettings {
     maxAngularAcceleration=1000
     maxAngularVelocity=3
     maxLinearAcceleration=700
     maxLinearVelocity=500
     velocityScaleZ=0
     worldAlpha=0.25
     externalLinearDragV=<0.2 0.2>
    }
   }
   n5 = Get Backbling Physics Alpha()
   n6 = "Transform (Modify) Bone - Bone: dyn_hair_bk_MSTR"(ComponentPose=@n4, Alpha=@n5, Translation="0,0,0", Rotation="-7,0,0", Scale="1,1,1") {
    rotationMode=BMM_Additive
    rotationSpace=BCS_BoneSpace
    alphaBoolBlend {blendOption=Linear}
    alphaScaleBiasClamp {bInterpResult=true}
    boneToModify {boneName=dyn_hair_bk_MSTR}
    rotation {pitch=-7}
   }
   n7 = "Transform (Modify) Bone - Bone: dyn_hair_bk_l_MSTR"(ComponentPose=@n6, Alpha=@n5, Translation="0,0,0", Rotation="-7,0,0", Scale="1,1,1") {
    rotationMode=BMM_Additive
    rotationSpace=BCS_BoneSpace
    alphaBoolBlend {blendOption=Linear}
    alphaScaleBiasClamp {bInterpResult=true}
    boneToModify {boneName=dyn_hair_bk_l_MSTR}
    rotation {pitch=-7}
   }
   n8 = "Transform (Modify) Bone - Bone: dyn_hair_bk_r_MSTR"(ComponentPose=@n7, Alpha=@n5, Translation="0,0,0", Rotation="-7,0,0", Scale="1,1,1") {
    rotationMode=BMM_Additive
    rotationSpace=BCS_BoneSpace
    alphaBoolBlend {blendOption=Linear}
    alphaScaleBiasClamp {bInterpResult=true}
    boneToModify {boneName=dyn_hair_bk_r_MSTR}
    rotation {pitch=-7}
   }
   n9 = "Constraint - dyn_hair_bk_01_l (PP)"!(ComponentPose=@n8, Alpha=1, ConstraintWeights_0=0.1, ConstraintWeights_1:float=0.9) -> (Pose) {
    alphaBoolBlend {blendOption=Linear}
    boneToModify {boneName=dyn_hair_bk_01_l}
    constraintSetup [{targetBone {boneName=neck_02}}, {targetBone {boneName=head}}]
    constraintWeights [1,1]
   }
   n10 = "Constraint - dyn_hair_bk_01_r (PP)"!(ComponentPose=@n9, Alpha=1, ConstraintWeights_0=0.1, ConstraintWeights_1:float=0.9) -> (Pose) {
    alphaBoolBlend {blendOption=Linear}
    boneToModify {boneName=dyn_hair_bk_01_r}
    constraintSetup [{targetBone {boneName=neck_02}}, {targetBone {boneName=head}}]
    constraintWeights [1,1]
   }
   n11 = "Constraint - dyn_hair_bk_01 (PP)"!(ComponentPose=@n10, Alpha=1, ConstraintWeights_0=0.1, ConstraintWeights_1:float=0.9) -> (Pose) {
    alphaBoolBlend {blendOption=Linear}
    boneToModify {boneName=dyn_hair_bk_01}
    constraintSetup [{targetBone {boneName=neck_02}}, {targetBone {boneName=head}}]
    constraintWeights [1,1]
   }
   n12 = "Constraint - dyn_hair_bk_02_l (PPP)"!(ComponentPose=@n11, Alpha=1, ConstraintWeights_0=0, ConstraintWeights_1:float=0.5, ConstraintWeights_2:float=0.5) -> (Pose) {
    alphaBoolBlend {blendOption=Linear}
    boneToModify {boneName=dyn_hair_bk_02_l}
    constraintSetup [{targetBone {boneName=spine_05}}, {targetBone {boneName=neck_02}}, {targetBone {boneName=clavicle_l}}]
    constraintWeights [1,1,1]
   }
   n13 = "Constraint - dyn_hair_bk_02_r (PPP)"!(ComponentPose=@n12, Alpha=1, ConstraintWeights_0=0, ConstraintWeights_1:float=0.5, ConstraintWeights_2:float=0.5) -> (Pose) {
    alphaBoolBlend {blendOption=Linear}
    boneToModify {boneName=dyn_hair_bk_02_r}
    constraintSetup [{targetBone {boneName=spine_05}}, {targetBone {boneName=neck_02}}, {targetBone {boneName=clavicle_r}}]
    constraintWeights [1,1,1]
   }
   n14 = "Constraint - dyn_hair_bk_02 (PP)"!(ComponentPose=@n13, Alpha=1, ConstraintWeights_0=0, ConstraintWeights_1:float=1) -> (Pose) {
    alphaBoolBlend {blendOption=Linear}
    boneToModify {boneName=dyn_hair_bk_02}
    constraintSetup [{targetBone {boneName=spine_05}}, {targetBone {boneName=neck_02}}]
    constraintWeights [1,1]
   }
   n15 = "Constraint - dyn_hair_bk_03_l (PP)"!(ComponentPose=@n14, Alpha=1, ConstraintWeights_0=0.5, ConstraintWeights_1:float=0.5) -> (Pose) {
    alphaBoolBlend {blendOption=Linear}
    boneToModify {boneName=dyn_hair_bk_03_l}
    constraintSetup [{targetBone {boneName=spine_04}}, {targetBone {boneName=spine_05}}]
    constraintWeights [1,1]
   }
   n16 = "Constraint - dyn_hair_bk_03_r (PP)"!(ComponentPose=@n15, Alpha=1, ConstraintWeights_0=0.5, ConstraintWeights_1:float=0.5) -> (Pose) {
    alphaBoolBlend {blendOption=Linear}
    boneToModify {boneName=dyn_hair_bk_03_r}
    constraintSetup [{targetBone {boneName=spine_04}}, {targetBone {boneName=spine_05}}]
    constraintWeights [1,1]
   }
   n17 = "Constraint - dyn_hair_bk_03 (PP)"!(ComponentPose=@n16, Alpha=1, ConstraintWeights_0=0.5, ConstraintWeights_1:float=0.5) -> (Pose) {
    alphaBoolBlend {blendOption=Linear}
    boneToModify {boneName=dyn_hair_bk_03}
    constraintSetup [{targetBone {boneName=spine_04}}, {targetBone {boneName=spine_05}}]
    constraintWeights [1,1]
   }
   n18 = "Constraint - dyn_hair_bk_04_l (PP)"!(ComponentPose=@n17, Alpha=1, ConstraintWeights_0=0.5, ConstraintWeights_1:float=0.5) -> (Pose) {
    alphaBoolBlend {blendOption=Linear}
    boneToModify {boneName=dyn_hair_bk_04_l}
    constraintSetup [{targetBone {boneName=spine_03}}, {targetBone {boneName=spine_04}}]
    constraintWeights [1,1]
   }
   n19 = "Constraint - dyn_hair_bk_04_r (PP)"!(ComponentPose=@n18, Alpha=1, ConstraintWeights_0=0.5, ConstraintWeights_1:float=0.5) -> (Pose) {
    alphaBoolBlend {blendOption=Linear}
    boneToModify {boneName=dyn_hair_bk_04_r}
    constraintSetup [{targetBone {boneName=spine_03}}, {targetBone {boneName=spine_04}}]
    constraintWeights [1,1]
   }
   n20 = "Constraint - dyn_hair_bk_04 (PP)"!(ComponentPose=@n19, Alpha=1, ConstraintWeights_0=0.5, ConstraintWeights_1:float=0.5) -> (Pose) {
    alphaBoolBlend {blendOption=Linear}
    boneToModify {boneName=dyn_hair_bk_04}
    constraintSetup [{targetBone {boneName=spine_03}}, {targetBone {boneName=spine_04}}]
    constraintWeights [1,1]
   }
   n21 = "Transform (Modify) Bone - Bone: dyn_hair_bk_02_l"(ComponentPose=@n20, Alpha=1, Translation="0,0,2.5", Rotation="0,0,5", Scale="1,1,1") {
    rotationMode=BMM_Replace
    rotationSpace=BCS_BoneSpace
    translationMode=BMM_Additive
    translationSpace=BCS_BoneSpace
    alphaBoolBlend {blendOption=Linear}
    alphaScaleBiasClamp {bInterpResult=true}
    boneToModify {boneName=dyn_hair_bk_02_l}
    rotation {roll=5}
    translation {z=2.5}
   }
   n22 = "Transform (Modify) Bone - Bone: dyn_hair_bk_02"(ComponentPose=@n21, Alpha=1, Translation="0,0,2.5", Rotation="-5,0,-10", Scale="1,1,1") {
    rotationMode=BMM_Replace
    rotationSpace=BCS_BoneSpace
    translationMode=BMM_Additive
    translationSpace=BCS_BoneSpace
    alphaBoolBlend {blendOption=Linear}
    alphaScaleBiasClamp {bInterpResult=true}
    boneToModify {boneName=dyn_hair_bk_02}
    rotation {pitch=-5, roll=-10}
    translation {z=2.5}
   }
   n23 = "Transform (Modify) Bone - Bone: dyn_hair_bk_02_r"(ComponentPose=@n22, Alpha=1, Translation="0,0,5", Rotation="-5,-5,-5", Scale="1,1,1") {
    rotationMode=BMM_Replace
    rotationSpace=BCS_BoneSpace
    translationMode=BMM_Additive
    translationSpace=BCS_BoneSpace
    alphaBoolBlend {blendOption=Linear}
    alphaScaleBiasClamp {bInterpResult=true}
    boneToModify {boneName=dyn_hair_bk_02_r}
    rotation=<rot -5 -5 -5>
    translation {z=5}
   }
   n24 = "Transform (Modify) Bone - Bone: dyn_hair_bk_03_l"(ComponentPose=@n23, Alpha=1, Translation="0,0,1.5", Rotation="-10,0,0", Scale="1,1,1") {
    rotationMode=BMM_Replace
    rotationSpace=BCS_BoneSpace
    translationMode=BMM_Additive
    translationSpace=BCS_BoneSpace
    alphaBoolBlend {blendOption=Linear}
    alphaScaleBiasClamp {bInterpResult=true}
    boneToModify {boneName=dyn_hair_bk_03_l}
    rotation {pitch=-10}
    translation {z=1.5}
   }
   n25 = "Transform (Modify) Bone - Bone: dyn_hair_bk_03"(ComponentPose=@n24, Alpha=1, Translation="0,0,-2", Rotation="5,5,0", Scale="1,1,1") {
    rotationMode=BMM_Replace
    rotationSpace=BCS_BoneSpace
    translationMode=BMM_Additive
    translationSpace=BCS_BoneSpace
    alphaBoolBlend {blendOption=Linear}
    alphaScaleBiasClamp {bInterpResult=true}
    boneToModify {boneName=dyn_hair_bk_03}
    rotation {pitch=5, yaw=5}
    translation {z=-2}
   }
   n26 = "Transform (Modify) Bone - Bone: dyn_hair_bk_03_r"(ComponentPose=@n25, Alpha=1, Translation="0,0,3.5", Rotation="-5,5,0", Scale="1,1,1") {
    rotationMode=BMM_Replace
    rotationSpace=BCS_BoneSpace
    translationMode=BMM_Additive
    translationSpace=BCS_BoneSpace
    alphaBoolBlend {blendOption=Linear}
    alphaScaleBiasClamp {bInterpResult=true}
    boneToModify {boneName=dyn_hair_bk_03_r}
    rotation {pitch=-5, yaw=5}
    translation {z=3.5}
   }
   n27 = "Transform (Modify) Bone - Bone: dyn_hair_bk_04_l"(ComponentPose=@n26, Alpha=1, Translation="0,0,0", Rotation="-10,0,0", Scale="1,1,1") {
    rotationMode=BMM_Replace
    rotationSpace=BCS_BoneSpace
    translationMode=BMM_Additive
    translationSpace=BCS_BoneSpace
    alphaBoolBlend {blendOption=Linear}
    alphaScaleBiasClamp {bInterpResult=true}
    boneToModify {boneName=dyn_hair_bk_04_l}
    rotation {pitch=-10}
   }
   n28 = "Transform (Modify) Bone - Bone: dyn_hair_bk_04"(ComponentPose=@n27, Alpha=1, Translation="0,0,-4", Rotation="10,0,0", Scale="1,1,1") {
    rotationMode=BMM_Replace
    rotationSpace=BCS_BoneSpace
    translationMode=BMM_Additive
    translationSpace=BCS_BoneSpace
    alphaBoolBlend {blendOption=Linear}
    alphaScaleBiasClamp {bInterpResult=true}
    boneToModify {boneName=dyn_hair_bk_04}
    rotation {pitch=10}
    translation {z=-4}
   }
   n29 = "Transform (Modify) Bone - Bone: dyn_hair_bk_04_r"(ComponentPose=@n28, Alpha=1, Translation="0,0,1.5", Rotation="-10,0,0", Scale="1,1,1") {
    rotationMode=BMM_Replace
    rotationSpace=BCS_BoneSpace
    translationMode=BMM_Additive
    translationSpace=BCS_BoneSpace
    alphaBoolBlend {blendOption=Linear}
    alphaScaleBiasClamp {bInterpResult=true}
    boneToModify {boneName=dyn_hair_bk_04_r}
    rotation {pitch=-10}
    translation {z=1.5}
   }
   n30 = Get body Mesh() -> (body Mesh:Skeletal Mesh Component Object Reference)
   n31 = Control Rig!(
    Source=Layered blend per bone(BasePose=Copy Pose From Mesh() {bCopyCurves=true, bUseAttachedParent=true}, BlendPoses_0=Copy Pose From Mesh(SourceMeshComponent=@n30)) {layerSetup [{branchFilters [{boneName=upperarm_l}, {boneName=upperarm_r}]}]}
    Alpha
    Is Targeting Pistol Float:double=Get Is Pistol Alpha() -> (Is Pistol Alpha:float)
    Is Targeting Rifle Float:double=Get Is Rifle Alpha() -> (Is Rifle Alpha:float)
    Is Skydive Float:double=Get SkyDive Float() -> (SkyDive Float:float)
   ) -> (Pose) {
    alphaBoolBlend {blendOption=Linear}
    controlRigAssetReference {blueprintRigClass=&F_MED_FuzzyBearPanda_Sparkle_FaceAcc_CtrlRig_C}
   }
   n32 = Control Rig!(Source=@n31, Alpha=Get Sky Dive Control Rig Alpha()) -> (Pose) {
    alphaBoolBlend {blendOption=Linear}
    controlRigAssetReference {blueprintRigClass=&F_MED_FuzzyBearPanda_Sparkle_FaceAccConstraint_CtrlRig_C}
   }
   n33 = Control Rig(Source=@n32, Alpha=Get Is Pistol Targeting Alpha()) {
    alphaBoolBlend {blendOption=Linear}
    controlRigAssetReference {blueprintRigClass=&F_MED_FuzzyBearPanda_Sparkle_FaceAcc_Pistol_CtrlRig_C}
   }
   n34 = Get Is Melee Float() -> (Is Melee Float:float)
   n35 = "Transform (Modify) Bone - Bone: dyn_hair_bk_l_MSTR"(ComponentPose=Local To Component(LocalPose=@n33), Alpha=@n34, Translation="0,0,0", Rotation="15,-5,-10", Scale="1,1,1") {
    rotationMode=BMM_Additive
    rotationSpace=BCS_BoneSpace
    alphaBoolBlend {blendOption=Linear}
    alphaScaleBiasClamp {bInterpResult=true, interpSpeedDecreasing=2, interpSpeedIncreasing=2}
    boneToModify {boneName=dyn_hair_bk_l_MSTR}
    rotation=<rot 15 -5 -10>
   }
   n36 = "Transform (Modify) Bone - Bone: dyn_hair_bk_MSTR"(ComponentPose=@n35, Alpha=@n34, Translation="0,0,0", Rotation="15,0,0", Scale="1,1,1") {
    rotationMode=BMM_Additive
    rotationSpace=BCS_BoneSpace
    alphaBoolBlend {blendOption=Linear}
    alphaScaleBiasClamp {bInterpResult=true, interpSpeedDecreasing=2, interpSpeedIncreasing=2}
    boneToModify {boneName=dyn_hair_bk_MSTR}
    rotation {pitch=15}
   }
   n37 = "Transform (Modify) Bone - Bone: dyn_hair_bk_r_MSTR"(ComponentPose=@n36, Alpha=@n34, Translation="0,0,0", Rotation="15,5,-10", Scale="1,1,1") {
    rotationMode=BMM_Additive
    rotationSpace=BCS_BoneSpace
    alphaBoolBlend {blendOption=Linear}
    alphaScaleBiasClamp {bInterpResult=true, interpSpeedDecreasing=2, interpSpeedIncreasing=2}
    boneToModify {boneName=dyn_hair_bk_r_MSTR}
    rotation=<rot 15 5 -10>
   }
   n38 = Get RigidBodyVector on Melee() -> (RigidBodyVector on Melee:Vector)
   n39 = RigidBody!(
    ComponentPose=@n37
    Alpha=Get PhysicsAlpha() -> (PhysicsAlpha:double)
    OverrideWorldGravity=Get GravityOverride()
    ExternalForce
    ComponentLinearAccScale:Vector=@n38
    ComponentLinearVelScale:Vector=@n38
   ) -> (Pose) {
    bClampLinearTranslationLimitToRefPose=true
    bOverrideWorldGravity=true
    overridePhysicsAsset=&/F_MED_FuzzyBearPanda_Sparkle_FaceAcc_Physics
    simulationSpace=BaseBoneSpace
    alphaBoolBlend {blendOption=Linear}
    alphaScaleBiasClamp {bInterpResult=true, interpSpeedIncreasing=2}
    baseBoneRef {boneName=spine_05}
    componentAppliedLinearAccClamp=<3000 3000 3000>
    componentLinearAccScale=<0.4 0.3 0.3>
    componentLinearVelScale=<0.4 0.3 0.3>
    simSpaceSettings {
     maxAngularAcceleration=1000
     maxAngularVelocity=3
     maxLinearAcceleration=700
     maxLinearVelocity=500
     velocityScaleZ=0
     worldAlpha=0.25
     externalLinearDragV=<0.2 0.2>
    }
   }
   n40 = Get SkyDive HairRot Alpha() -> (SkyDive HairRot Alpha:float)
   n41 = "Transform (Modify) Bone - Bone: dyn_hair_bk_l_MSTR"(ComponentPose=@n39, Alpha=@n40, Translation="0,0,2.5", Rotation="7.5,0,0", Scale="1,1,1") {
    rotationMode=BMM_Additive
    rotationSpace=BCS_BoneSpace
    translationMode=BMM_Additive
    translationSpace=BCS_BoneSpace
    alphaBoolBlend {blendOption=Linear}
    alphaScaleBiasClamp {bInterpResult=true}
    boneToModify {boneName=dyn_hair_bk_l_MSTR}
    rotation {pitch=7.5}
    translation {z=2.5}
   }
   n42 = "Transform (Modify) Bone - Bone: dyn_hair_bk_MSTR"(ComponentPose=@n41, Alpha=@n40, Translation="0,0,2.5", Rotation="7.5,0,0", Scale="1,1,1") {
    rotationMode=BMM_Additive
    rotationSpace=BCS_BoneSpace
    translationMode=BMM_Additive
    translationSpace=BCS_BoneSpace
    alphaBoolBlend {blendOption=Linear}
    alphaScaleBiasClamp {bInterpResult=true}
    boneToModify {boneName=dyn_hair_bk_MSTR}
    rotation {pitch=7.5}
    translation {z=2.5}
   }
   n43 = "Transform (Modify) Bone - Bone: dyn_hair_bk_r_MSTR"(ComponentPose=@n42, Alpha=@n40, Translation="0,0,2.5", Rotation="7.5,0,0", Scale="1,1,1") {
    rotationMode=BMM_Additive
    rotationSpace=BCS_BoneSpace
    translationMode=BMM_Additive
    translationSpace=BCS_BoneSpace
    alphaBoolBlend {blendOption=Linear}
    alphaScaleBiasClamp {bInterpResult=true}
    boneToModify {boneName=dyn_hair_bk_r_MSTR}
    rotation {pitch=7.5}
    translation {z=2.5}
   }
   n44 = Get SkyDive HairRot Alpha()
   n45 = "Transform (Modify) Bone - Bone: dyn_hair_bk_01_l"(ComponentPose=@n43, Alpha=@n44, Translation="0,0,0", Rotation="0,0,0", Scale="1,1,1") {
    rotationMode=BMM_Replace
    rotationSpace=BCS_BoneSpace
    translationMode=BMM_Additive
    translationSpace=BCS_BoneSpace
    alphaBoolBlend {blendOption=Linear}
    alphaScaleBiasClamp {bInterpResult=true, interpSpeedIncreasing=2}
    boneToModify {boneName=dyn_hair_bk_01_l}
   }
   n46 = "Transform (Modify) Bone - Bone: dyn_hair_bk_01"(ComponentPose=@n45, Alpha=@n44, Translation="0,0,0", Rotation="0,0,0", Scale="1,1,1") {
    rotationMode=BMM_Replace
    rotationSpace=BCS_BoneSpace
    translationMode=BMM_Additive
    translationSpace=BCS_BoneSpace
    alphaBoolBlend {blendOption=Linear}
    alphaScaleBiasClamp {bInterpResult=true, interpSpeedIncreasing=2}
    boneToModify {boneName=dyn_hair_bk_01}
   }
   n47 = "Transform (Modify) Bone - Bone: dyn_hair_bk_01_r"(ComponentPose=@n46, Alpha=@n44, Translation="0,0,0", Rotation="0,0,0", Scale="1,1,1") {
    rotationMode=BMM_Replace
    rotationSpace=BCS_BoneSpace
    translationMode=BMM_Additive
    translationSpace=BCS_BoneSpace
    alphaBoolBlend {blendOption=Linear}
    alphaScaleBiasClamp {bInterpResult=true, interpSpeedIncreasing=2}
    boneToModify {boneName=dyn_hair_bk_01_r}
   }
   n48 = "Transform (Modify) Bone - Bone: dyn_hair_bk_02_l"(ComponentPose=@n47, Alpha=@n44, Translation="0,0,-5", Rotation="-10,0,0", Scale="1,1,1") {
    rotationMode=BMM_Replace
    rotationSpace=BCS_BoneSpace
    translationMode=BMM_Additive
    translationSpace=BCS_BoneSpace
    alphaBoolBlend {blendOption=Linear}
    alphaScaleBiasClamp {bInterpResult=true, interpSpeedIncreasing=2}
    boneToModify {boneName=dyn_hair_bk_02_l}
    rotation {pitch=-10}
    translation {z=-5}
   }
   n49 = "Transform (Modify) Bone - Bone: dyn_hair_bk_02"(ComponentPose=@n48, Alpha=@n44, Translation="0,0,-5", Rotation="-10,0,0", Scale="1,1,1") {
    rotationMode=BMM_Replace
    rotationSpace=BCS_BoneSpace
    translationMode=BMM_Additive
    translationSpace=BCS_BoneSpace
    alphaBoolBlend {blendOption=Linear}
    alphaScaleBiasClamp {bInterpResult=true, interpSpeedIncreasing=2}
    boneToModify {boneName=dyn_hair_bk_02}
    rotation {pitch=-10}
    translation {z=-5}
   }
   n50 = "Transform (Modify) Bone - Bone: dyn_hair_bk_02_r"(ComponentPose=@n49, Alpha=@n44, Translation="0,0,-5", Rotation="-10,0,0", Scale="1,1,1") {
    rotationMode=BMM_Replace
    rotationSpace=BCS_BoneSpace
    translationMode=BMM_Additive
    translationSpace=BCS_BoneSpace
    alphaBoolBlend {blendOption=Linear}
    alphaScaleBiasClamp {bInterpResult=true, interpSpeedIncreasing=2}
    boneToModify {boneName=dyn_hair_bk_02_r}
    rotation {pitch=-10}
    translation {z=-5}
   }
   n51 = Get Is Falling Jumping Alpha() -> (Is Falling Jumping Alpha:double)
   n52 = "Transform (Modify) Bone - Bone: dyn_hair_bk_03_l"(ComponentPose=@n50, Alpha=@n51, Translation="0,0,1.5", Rotation="5,0,0", Scale="1,1,1") {
    rotationMode=BMM_Additive
    rotationSpace=BCS_BoneSpace
    translationMode=BMM_Additive
    translationSpace=BCS_BoneSpace
    alphaBoolBlend {blendOption=Linear}
    alphaScaleBiasClamp {bInterpResult=true, interpSpeedDecreasing=20, interpSpeedIncreasing=3}
    boneToModify {boneName=dyn_hair_bk_03_l}
    rotation {pitch=5}
    translation {z=1.5}
   }
   n53 = "Transform (Modify) Bone - Bone: dyn_hair_bk_04_l"(ComponentPose=@n52, Alpha=@n51, Translation="0,0,0", Rotation="15,0,0", Scale="1,1,1") {
    rotationMode=BMM_Additive
    rotationSpace=BCS_BoneSpace
    translationMode=BMM_Additive
    translationSpace=BCS_BoneSpace
    alphaBoolBlend {blendOption=Linear}
    alphaScaleBiasClamp {bInterpResult=true, interpSpeedDecreasing=20, interpSpeedIncreasing=3}
    boneToModify {boneName=dyn_hair_bk_04_l}
    rotation {pitch=15}
   }
   n54 = "Transform (Modify) Bone - Bone: dyn_hair_bk_05_l"(ComponentPose=@n53, Alpha=@n51, Translation="0,0,0", Rotation="30,0,0", Scale="1,1,1") {
    rotationMode=BMM_Additive
    rotationSpace=BCS_BoneSpace
    translationMode=BMM_Additive
    translationSpace=BCS_BoneSpace
    alphaBoolBlend {blendOption=Linear}
    alphaScaleBiasClamp {bInterpResult=true, interpSpeedDecreasing=20, interpSpeedIncreasing=3}
    boneToModify {boneName=dyn_hair_bk_05_l}
    rotation {pitch=30}
   }
   n55 = "Transform (Modify) Bone - Bone: dyn_hair_bk_06_l"(ComponentPose=@n54, Alpha=@n51, Translation="0,0,0", Rotation="25,0,0", Scale="1,1,1") {
    rotationMode=BMM_Additive
    rotationSpace=BCS_BoneSpace
    translationMode=BMM_Additive
    translationSpace=BCS_BoneSpace
    alphaBoolBlend {blendOption=Linear}
    alphaScaleBiasClamp {bInterpResult=true, interpSpeedDecreasing=20, interpSpeedIncreasing=3}
    boneToModify {boneName=dyn_hair_bk_06_l}
    rotation {pitch=25}
   }
   n56 = "Transform (Modify) Bone - Bone: dyn_hair_bk_03"(ComponentPose=@n55, Alpha=@n51, Translation="0,0,1.5", Rotation="5,0,0", Scale="1,1,1") {
    rotationMode=BMM_Additive
    rotationSpace=BCS_BoneSpace
    translationMode=BMM_Additive
    translationSpace=BCS_BoneSpace
    alphaBoolBlend {blendOption=Linear}
    alphaScaleBiasClamp {bInterpResult=true, interpSpeedDecreasing=20, interpSpeedIncreasing=3}
    boneToModify {boneName=dyn_hair_bk_03}
    rotation {pitch=5}
    translation {z=1.5}
   }
   n57 = "Transform (Modify) Bone - Bone: dyn_hair_bk_04"(ComponentPose=@n56, Alpha=@n51, Translation="0,0,0", Rotation="25,0,0", Scale="1,1,1") {
    rotationMode=BMM_Additive
    rotationSpace=BCS_BoneSpace
    translationMode=BMM_Additive
    translationSpace=BCS_BoneSpace
    alphaBoolBlend {blendOption=Linear}
    alphaScaleBiasClamp {bInterpResult=true, interpSpeedDecreasing=20, interpSpeedIncreasing=3}
    boneToModify {boneName=dyn_hair_bk_04}
    rotation {pitch=25}
   }
   n58 = "Transform (Modify) Bone - Bone: dyn_hair_bk_05"(ComponentPose=@n57, Alpha=@n51, Translation="0,0,0", Rotation="30,0,0", Scale="1,1,1") {
    rotationMode=BMM_Additive
    rotationSpace=BCS_BoneSpace
    translationMode=BMM_Additive
    translationSpace=BCS_BoneSpace
    alphaBoolBlend {blendOption=Linear}
    alphaScaleBiasClamp {bInterpResult=true, interpSpeedDecreasing=20, interpSpeedIncreasing=3}
    boneToModify {boneName=dyn_hair_bk_05}
    rotation {pitch=30}
   }
   n59 = "Transform (Modify) Bone - Bone: dyn_hair_bk_06"(ComponentPose=@n58, Alpha=@n51, Translation="0,0,0", Rotation="25,0,0", Scale="1,1,1") {
    rotationMode=BMM_Additive
    rotationSpace=BCS_BoneSpace
    translationMode=BMM_Additive
    translationSpace=BCS_BoneSpace
    alphaBoolBlend {blendOption=Linear}
    alphaScaleBiasClamp {bInterpResult=true, interpSpeedDecreasing=20, interpSpeedIncreasing=3}
    boneToModify {boneName=dyn_hair_bk_06}
    rotation {pitch=25}
   }
   n60 = "Transform (Modify) Bone - Bone: dyn_hair_bk_03_r"(ComponentPose=@n59, Alpha=@n51, Translation="0,0,1.5", Rotation="5,0,0", Scale="1,1,1") {
    rotationMode=BMM_Additive
    rotationSpace=BCS_BoneSpace
    translationMode=BMM_Additive
    translationSpace=BCS_BoneSpace
    alphaBoolBlend {blendOption=Linear}
    alphaScaleBiasClamp {bInterpResult=true, interpSpeedDecreasing=20, interpSpeedIncreasing=3}
    boneToModify {boneName=dyn_hair_bk_03_r}
    rotation {pitch=5}
    translation {z=1.5}
   }
   n61 = "Transform (Modify) Bone - Bone: dyn_hair_bk_04_r"(ComponentPose=@n60, Alpha=@n51, Translation="0,0,0", Rotation="25,0,0", Scale="1,1,1") {
    rotationMode=BMM_Additive
    rotationSpace=BCS_BoneSpace
    translationMode=BMM_Additive
    translationSpace=BCS_BoneSpace
    alphaBoolBlend {blendOption=Linear}
    alphaScaleBiasClamp {bInterpResult=true, interpSpeedDecreasing=20, interpSpeedIncreasing=3}
    boneToModify {boneName=dyn_hair_bk_04_r}
    rotation {pitch=25}
   }
   n62 = "Transform (Modify) Bone - Bone: dyn_hair_bk_05_r"(ComponentPose=@n61, Alpha=@n51, Translation="0,0,0", Rotation="30,0,0", Scale="1,1,1") {
    rotationMode=BMM_Additive
    rotationSpace=BCS_BoneSpace
    translationMode=BMM_Additive
    translationSpace=BCS_BoneSpace
    alphaBoolBlend {blendOption=Linear}
    alphaScaleBiasClamp {bInterpResult=true, interpSpeedDecreasing=20, interpSpeedIncreasing=3}
    boneToModify {boneName=dyn_hair_bk_05_r}
    rotation {pitch=30}
   }
   n63 = "Transform (Modify) Bone - Bone: dyn_hair_bk_06_r"(ComponentPose=@n62, Alpha=@n51, Translation="0,0,0", Rotation="25,0,0", Scale="1,1,1") {
    rotationMode=BMM_Additive
    rotationSpace=BCS_BoneSpace
    translationMode=BMM_Additive
    translationSpace=BCS_BoneSpace
    alphaBoolBlend {blendOption=Linear}
    alphaScaleBiasClamp {bInterpResult=true, interpSpeedDecreasing=20, interpSpeedIncreasing=3}
    boneToModify {boneName=dyn_hair_bk_06_r}
    rotation {pitch=25}
   }
   n64 = Control Rig!(Source=Component To Local(ComponentPose=@n63), Alpha=Get Is Pistol Alpha(), Is Pistol Targeting Float:double=Get Is Pistol Targeting Alpha()) -> (Pose) {
    alphaBoolBlend {blendOption=Linear}
    alphaScaleBiasClamp {bInterpResult=true, interpSpeedDecreasing=2, interpSpeedIncreasing=0.5}
    controlRigAssetReference {blueprintRigClass=&/F_MED_FuzzyBearPanda_Sparkle_FaceAcc_PostSim_CtrlRig.F_MED_FuzzyBearPanda_Sparkle_FaceAcc_PostSim_CtrlRig_C}
   }
   Output Pose(Result=Blend Poses by bool(BlendPose_0=Component To Local(ComponentPose=@n29), BlendPose_1=@n64, bActiveValue=Get isWearingBackbling()) {blendType=Linear}) {layerGroup=None}
   Get SkyDive Alpha() -> (SkyDive Alpha:double)
  ]}
 EventGraph {nodes [
   Delay(Duration=0.1)
   > Do Once()
   Event BlueprintUpdateAnimation()
   > n1 = Sequence!(execute) -> (then_0, then_1, then_2:Exec, then_3:Exec, then_4:Exec, then_5:Exec, then_6:Exec)
   > n2 = Gravity Override State Function DEPRECATED(CharacterPartAnimInstance=Self-Reference(), GameTimeInSeconds=GetGameTimeInSeconds(), GravityOverrideParams=Get Gravity Override Params())
   > Set GravityOverride(GravityOverride=@n2)
   n3 = Set headRotBack(
    execute=@n1.then_1
    headRotBack:double=GetAlphaFromReferenceTransformDelta(SkeletalMeshComponent=GetOwnerSkeletalMeshComponent(), BoneName=head, BaseName=spine_05, InputRange=(X=20,Y=-20), TwistAxis="1,0,0")
   ) -> (Output_Get:double)
   > n4 = Branch(Condition=Get bIsPlayingMeleeAnim())
   > Set MeleeHeadDwn(MeleeHeadDwn:double=@n3) -> (Output_Get:double)
   Set MeleeHeadDwn(execute=@n4.else, MeleeHeadDwn=1)
   n5 = Sequence!(execute=@n1.then_2) -> (then_0, then_1, then_2, then_3)
   > n6 = Cast To FortPlayerAnimInstance(Object=GetAnimInstance(self=GetOwnerSkeletalMeshComponent()))
   > Set isPistol(isPistol=false)
   > Set isRifle(isRifle:bool=false) -> (Output_Get:bool)
   > n7 = Switch on EFortWeaponCoreAnimation!(execute, Selection:EFortWeaponCoreAnimation Enum=Get WeaponCoreAnim(self=@n6)) -> (
    Melee:Exec
    Pistol:Exec
    Shotgun:Exec
    PaperBlueprint:Exec
    Rifle:Exec
    MeleeOneHand:Exec
    MachinePistol:Exec
    RocketLauncher:Exec
    GrenadeLauncher:Exec
    GoingCommando:Exec
    AssaultRifle:Exec
    TacticalShotgun:Exec
    SniperRifle:Exec
    TrapPlacement:Exec
    ShoulderLauncher:Exec
    AbilityDecoTool:Exec
    Crossbow:Exec
    C4:Exec
    RemoteControl:Exec
    DualWield:Exec
    AR_BullPup:Exec
    AR_ForwardGrip:Exec
    MedPackPaddles:Exec
    SMG_P90:Exec
    AR_DrumGun:Exec
    Consumable_Small:Exec
    Consumable_Large:Exec
    Balloon:Exec
    MountedTurret:Exec
    CreativeTool:Exec
    ExplosiveBow:Exec
    AshtonIndigo:Exec
    AshtonChicago:Exec
    MeleeDualWield:Exec
    Unarmed:Exec
   )
   Set isRifle(execute=[@n7.AssaultRifle, @n7.Rifle, @n7.RocketLauncher, @n7.Shotgun, @n7.SniperRifle, @n7.TacticalShotgun], isRifle=true)
   Set isPistol(execute=@n7.Pistol, isPistol=true)
   Set Is Rifle Alpha(execute=@n5.then_1, Is Rifle Alpha:float="To Float (Boolean)"(InBool=Get isRifle() -> (isRifle:bool))) -> (Output_Get:float)
   Set Is Pistol Alpha(execute=@n5.then_2, Is Pistol Alpha:float="To Float (Boolean)"(InBool=Get isPistol() -> (isPistol:bool))) -> (Output_Get:float)
   n8 = Get isPistol()
   n9 = Get bIsTargeting()
   Set Is Pistol Targeting Alpha(execute=@n5.then_3, Is Pistol Targeting Alpha:float="To Float (Boolean)"(InBool=AND Boolean(A=@n8, B=@n9))) -> (Output_Get:float)
   Set body Mesh(execute=@n1.then_3, body Mesh:Skeletal Mesh Component Object Reference=GetOwnerSkeletalMeshComponent()) -> (Output_Get:Skeletal Mesh Component Object Reference)
   n10 = Sequence!(execute=@n1.then_4) -> (then_0, then_1, then_2, then_3, then_4:Exec, then_5:Exec)
   > Set PhysicsAlpha(PhysicsAlpha:double=float * float!(
     A=1
     B=Get SkyDive Float()
     C:double=Get Is Tactical Sprinting Float() -> (Is Tactical Sprinting Float:double)
     D:float=Get Is Melee Alpha() -> (Is Melee Alpha:float)
     E:float=Get Not Backbling Alpha() -> (Not Backbling Alpha:float)
     F:float=Get Falling Alpha() -> (Falling Alpha:float)
    ) -> (ReturnValue) {is=title K2Node_PromotableOperator}) -> (Output_Get:double)
   n11 = Branch(execute=@n10.then_1, Condition=Get bIsSkyDiving())
   > Set SkyDive Float(SkyDive Float:float=0.5) -> (Output_Get:float)
   Set SkyDive Float(execute=@n11.else, SkyDive Float=1)
   n12 = Get bIsSkyDiving()
   n13 = Get bIsFalling()
   n14 = Get bIsJumping()
   n15 = Branch(execute=@n10.then_2, Condition=AND Boolean(A=OR Boolean(A=@n14, B=@n13), B=NOT Boolean(A=@n12)))
   > Set Falling Alpha(Falling Alpha:float=0.65) -> (Output_Get:float)
   Set Falling Alpha(execute=@n15.else, Falling Alpha=1)
   n16 = Branch(execute=@n10.then_3, Condition=Get bIsPlayingMeleeAnim())
   > Set Is Melee Alpha(Is Melee Alpha:float=1) -> (Output_Get:float)
   Set Is Melee Alpha(execute=@n16.else, Is Melee Alpha=1)
   n17 = Get isWearingBackbling()
   Set Not Backbling Alpha(execute=@n10.then_4, Not Backbling Alpha:float="To Float (Boolean)"(InBool=NOT Boolean(A=@n17))) -> (Output_Get:float)
   n18 = Branch(execute=@n10.then_5, Condition=Get bIsTacticalSprinting())
   > Set Is Tactical Sprinting Float(Is Tactical Sprinting Float:double=0.75) -> (Output_Get:double)
   Set Is Tactical Sprinting Float(execute=@n18.else, Is Tactical Sprinting Float=1)
   n19 = Sequence!(execute=@n1.then_5) -> (then_0, then_1, then_2, then_3, then_4, then_5, then_6:Exec)
   > Set Is Falling Jumping Alpha(Is Falling Jumping Alpha:double=float - float(A=Get Is Falling Jumping Float() -> (Is Falling Jumping Float:float), B=Get Is Sky Diving() -> (Is Sky Diving:double))) -> (Output_Get:double)
   Set SkyDive Alpha(
    execute=@n19.then_1
    SkyDive Alpha:double=float - float(A=Get Is Sky Diving(), B=Get Is Skydive Dive Mode() -> (Is Skydive Dive Mode:double))
   ) -> (Output_Get:double)
   n20 = NOT Boolean(A=Get bIsParachuteOpened())
   n21 = NOT Boolean(A=Get bIsSkydiveDiveMode())
   n22 = Get bIsSkyDiving()
   Set SkyDive HairRot Alpha(execute=@n19.then_2, SkyDive HairRot Alpha:float="To Float (Boolean)"(InBool=AND Boolean!(A=@n21, B=@n20, C:bool=@n22) -> (ReturnValue))) -> (Output_Get:float)
   n23 = Branch(execute=@n19.then_3, Condition=Get bIsSkyDiving())
   > Set Sky Dive Control Rig Alpha(Sky Dive Control Rig Alpha:double=0) -> (Output_Get:double)
   Set Sky Dive Control Rig Alpha(execute=@n23.else, Sky Dive Control Rig Alpha=1)
   Set Is Sky Diving(execute=@n19.then_4, Is Sky Diving:double="To Float (Boolean)"(InBool=Get bIsSkyDiving())) -> (Output_Get:double)
   Set Is Skydive Dive Mode(execute=@n19.then_5, Is Skydive Dive Mode:double="To Float (Boolean)"(InBool=Get bIsSkydiveDiveMode())) -> (Output_Get:double)
   n24 = Get bIsSprinting()
   n25 = NOT Boolean(A=AND Boolean!(A=NOT Boolean(A=@n24), B=Get bIsMoving2D()) -> (ReturnValue))
   n26 = OR Boolean!(A=false, B=Get bIsFalling(), C:bool) -> (ReturnValue)
   Set Is Falling Jumping Float(execute=@n19.then_6, Is Falling Jumping Float:float="To Float (Boolean)"(InBool=AND Boolean(A=@n26, B=@n25))) -> (Output_Get:float)
   n27 = Sequence!(execute=@n1.then_6) -> (then_0, then_1)
   > Set Is Melee Float(Is Melee Float:float="To Float (Boolean)"(InBool=Get bIsPlayingMeleeAnim())) -> (Output_Get:float)
   n28 = Branch(execute=@n27.then_1, Condition=Get bIsPlayingMeleeAnim())
   > Set RigidBodyVector on Melee(RigidBodyVector on Melee:Vector="0.1,0.1,0.1") -> (Output_Get:Vector)
   Set RigidBodyVector on Melee(execute=@n28.else, RigidBodyVector on Melee="0.4,0.3,0.3")
   Event OnCharacterCustomizationLoaded()
   > n29 = Sequence!(execute) -> (then_0, then_1, then_2:Exec)
   > Cast To PlayerPawn_Athena(Object=TryGetPawnOwner())
   n30 = CheckForBackbling()
   n31 = Set isWearingBackbling(execute=@n29.then_1, isWearingBackbling=OR Boolean!(A=@n30, B=@n30.bIsBackblingLong) -> (ReturnValue))
   > n32 = Sequence!(execute) -> (then_0, then_1)
   > Set HairRot(HairRot:float=1) -> (Output_Get:float)
   > Delay(Duration=0.5)
   > Set HairRot(HairRot=0)
   Set Backbling Physics Alpha(execute=@n32.then_1, Backbling Physics Alpha:float="To Float (Boolean)"(InBool=@n31)) -> (Output_Get:float)
   Event BlueprintInitializeAnimation()
   > Set body Mesh(body Mesh=GetOwnerSkeletalMeshComponent())
   Get bIsJumping()
  ]}
]
variables [
 Backbling Physics Alpha {type=number}
 body Mesh {type=SkeletalMeshComponent}
 Falling Alpha {type=number}
 Gravity Override Params {type=GravityOverrideParamsStruct}
 GravityOverride {type=Vector}
 HairRot {type=number}
 headRotBack {type=number}
 Is Falling Jumping Alpha {type=number}
 Is Falling Jumping Float {type=number}
 Is Melee Alpha {type=number}
 Is Melee Float {type=number}
 Is Pistol Alpha {type=number}
 Is Pistol Targeting Alpha {type=number}
 Is Rifle Alpha {type=number}
 Is Sky Diving {type=number}
 Is Skydive Dive Mode {type=number}
 Is Tactical Sprinting Float {type=number}
 isPistol {type=boolean}
 isRifle {type=boolean}
 isWearingBackbling {type=boolean}
 MeleeHeadDwn {type=number}
 Not Backbling Alpha {type=number}
 pawnVelocitySpeed {type=number}
 pawnVelocitySpeedCutoff {type=number}
 PhysicsAlpha {type=number}
 RigidBodyVector on Melee {type=Vector}
 Shake Frequency {type=number}
 Sky Dive Control Rig Alpha {type=number}
 SkyDive Alpha {type=number}
 SkyDive Float {type=number}
 SkyDive HairRot Alpha {type=number}
]
